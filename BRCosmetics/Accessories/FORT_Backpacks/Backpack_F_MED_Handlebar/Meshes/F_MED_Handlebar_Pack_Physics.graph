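# Physics asset for the F_MED_Handlebar cosmetic backpack, defining collision shapes and joint limits for its dyn-prefixed bones.
# Two bone capsules and a head capsule are constrained to the pack capsule with tight swing and twist limits.
# Two small eye spheres chain off the head with wider swing but very little twist, allowing them to wobble relative to it.
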PhysicsAsset /BRCosmetics/Accessories/FORT_Backpacks/Backpack_F_MED_Handlebar/Meshes/F_MED_Handlebar_Pack_Physics
bodies {
 dyn_bone_a [{shapeType=Capsule, length=11.8992, radius=1.0613, shapeName=dyn_bone_a_capsule, center=<0 0.2 -1>, rotation=<rot 0.0003 -178.4479 -42.5308>}]
 dyn_bone_b [{shapeType=Capsule, length=14.1803, radius=1.0608, shapeName=dyn_bone_b_capsule, center=<0 -0.2 -1>, rotation=<rot -0.0165 -0.127 -43.956>}]
 dyn_eye_l [{shapeType=Sphere, radius=1.1396, shapeName=dyn_eye_l_sphere, center=<0 -0.7125 -3.6954>}]
 dyn_eye_r [{shapeType=Sphere, radius=1.1396, shapeName=dyn_eye_r_sphere, center=<0 0.6166 -3.6954>}]
 dyn_head [{
   shapeType=Capsule
   length=7.475
   radius=4.2653
   shapeName=dyn_head_capsule
   center=<2.7951 -0.0062 -0.2511>
   rotation=<rot -1.4148 -90 -64.5086>
  }]
 pack [{shapeType=Capsule, length=7.5368, radius=2, center=<0 4.122 0>}]
}
constraints [
 {bone1Name=dyn_bone_a, bone2Name=pack, swing1LimitDegrees=7, swing2LimitDegrees=7, twistLimitDegrees=8}
 {bone1Name=dyn_bone_b, bone2Name=pack, swing1LimitDegrees=7, swing2LimitDegrees=7, twistLimitDegrees=8}
 {bone1Name=dyn_head, bone2Name=pack, swing1LimitDegrees=15, swing2LimitDegrees=15, twistLimitDegrees=5}
 {bone1Name=dyn_eye_l, bone2Name=dyn_head, swing1LimitDegrees=25, swing2LimitDegrees=25, twistLimitDegrees=2}
 {bone1Name=dyn_eye_r, bone2Name=dyn_head, swing1LimitDegrees=25, swing2LimitDegrees=25, twistLimitDegrees=2}
]
