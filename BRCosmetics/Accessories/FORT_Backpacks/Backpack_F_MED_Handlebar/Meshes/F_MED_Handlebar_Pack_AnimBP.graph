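# Anim Blueprint for the Handlebar backpack cosmetic that plays a procedural head-and-eye reaction when the player's kill stat updates.
# RigidBody physics feeds additive bone transforms, with staggered bool-alpha blends that lift, tilt and enlarge the head and eyes.
# Game time drives an oscillating head yaw and counter-spinning eye rolls.
# WatchCosmeticStat on AthenaKills sets three flags, which update-loop delays clear, head first, then eyes.
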AnimBlueprint /BRCosmetics/Accessories/FORT_Backpacks/Backpack_F_MED_Handlebar/Meshes/F_MED_Handlebar_Pack_AnimBP
target_skeleton=&/F_MED_Handlebar_Pack_Skeleton
graphs [
 AnimGraph {nodes [
   n1 = RigidBody(ComponentPose=Local To Component(LocalPose=Input Pose())) {
    bOverrideWorldGravity=true
    overridePhysicsAsset=&/F_MED_Handlebar_Pack_Physics
    simulationSpace=BaseBoneSpace
    alphaBoolBlend {blendOption=Linear}
    baseBoneRef {boneName=root}
    componentLinearAccScale=<0.6 0.6 0.6>
    componentLinearVelScale=<0.6 0.6 0.6>
    overrideWorldGravity {z=-980}
   }
   n2 = "Transform (Modify) Bone - Bone: dyn_eye_l"(ComponentPose=@n1, Alpha=1, Translation="0,0,0", Rotation="0,-12,0", Scale="1,1,1") {
    rotationMode=BMM_Additive
    rotationSpace=BCS_BoneSpace
    scaleMode=BMM_Additive
    scaleSpace=BCS_BoneSpace
    translationMode=BMM_Additive
    translationSpace=BCS_BoneSpace
    alphaBoolBlend {blendOption=Linear}
    boneToModify {boneName=dyn_eye_l}
    rotation {yaw=-12}
   }
   n3 = "Transform (Modify) Bone - Bone: dyn_eye_r"(ComponentPose=@n2, Alpha=1, Translation="0,0,0", Rotation="0,12,0", Scale="1,1,1") {
    rotationMode=BMM_Additive
    rotationSpace=BCS_BoneSpace
    scaleMode=BMM_Additive
    scaleSpace=BCS_BoneSpace
    translationMode=BMM_Additive
    translationSpace=BCS_BoneSpace
    alphaBoolBlend {blendOption=Linear}
    boneToModify {boneName=dyn_eye_r}
    rotation {yaw=12}
   }
   n4 = Get Active_Kill() -> (Active_Kill:bool)
   n5 = "Transform (Modify) Bone - Bone: dyn_head"!(ComponentPose=@n3, bAlphaBoolEnabled:bool=@n4, Translation="3,0,0", Rotation="0,0,0", Scale="1.1,1.1,1.1") -> (Pose) {
    alphaInputType=Bool
    rotationMode=BMM_Additive
    rotationSpace=BCS_BoneSpace
    scaleMode=BMM_Additive
    scaleSpace=BCS_BoneSpace
    translationMode=BMM_Additive
    translationSpace=BCS_BoneSpace
    alphaBoolBlend {blendInTime=0.1, blendOption=Cubic, blendOutTime=1.5}
    boneToModify {boneName=dyn_head}
    scale=<1.1 1.1 1.1>
    translation {x=3}
   }
   n6 = "Transform (Modify) Bone - Bone: dyn_head"(ComponentPose=@n5, bAlphaBoolEnabled=@n4, Translation="0,0,6", Rotation="Get rotate  Head"() -> ("rotate  Head":Rotator), Scale="1,1,1") {
    alphaInputType=Bool
    rotationMode=BMM_Additive
    rotationSpace=BCS_BoneSpace
    scaleMode=BMM_Additive
    scaleSpace=BCS_BoneSpace
    translationMode=BMM_Additive
    translationSpace=BCS_BoneSpace
    alphaBoolBlend {blendInTime=1, blendOption=Cubic, blendOutTime=2}
    boneToModify {boneName=dyn_head}
    translation {z=6}
   }
   n7 = "Transform (Modify) Bone - Bone: dyn_head"(ComponentPose=@n6, bAlphaBoolEnabled=@n4, Translation="0,0,2", Rotation="-20,0,0", Scale="1,1,1") {
    alphaInputType=Bool
    rotationMode=BMM_Additive
    rotationSpace=BCS_BoneSpace
    scaleMode=BMM_Additive
    scaleSpace=BCS_BoneSpace
    translationMode=BMM_Additive
    translationSpace=BCS_BoneSpace
    alphaBoolBlend {blendInTime=0.2, blendOption=Cubic, blendOutTime=1}
    boneToModify {boneName=dyn_head}
    rotation {pitch=-20}
    translation {z=2}
   }
   n8 = "Transform (Modify) Bone - Bone: dyn_head"(ComponentPose=@n7, bAlphaBoolEnabled=@n4, Translation="0,0,0", Rotation="45,0,0", Scale="1,1,1") {
    alphaInputType=Bool
    rotationMode=BMM_Additive
    rotationSpace=BCS_BoneSpace
    scaleMode=BMM_Additive
    scaleSpace=BCS_BoneSpace
    translationMode=BMM_Additive
    translationSpace=BCS_BoneSpace
    alphaBoolBlend {blendInTime=0.8, blendOption=Cubic, blendOutTime=1}
    boneToModify {boneName=dyn_head}
    rotation {pitch=45}
   }
   n9 = Get Active_Kill_C() -> (Active_Kill_C:bool)
   n10 = "Transform (Modify) Bone - Bone: dyn_eye_r"!(
    ComponentPose=@n8
    bAlphaBoolEnabled:bool=@n9
    Translation="0,0,0"
    Rotation="Get rotate  Eye  R"() -> ("rotate  Eye  R":Rotator)
    Scale="1,1,1"
   ) -> (Pose) {
    alphaInputType=Bool
    rotationMode=BMM_Additive
    rotationSpace=BCS_ParentBoneSpace
    scaleMode=BMM_Additive
    scaleSpace=BCS_BoneSpace
    translationMode=BMM_Additive
    translationSpace=BCS_BoneSpace
    alphaBoolBlend {blendInTime=0.3, blendOption=Cubic, blendOutTime=0.5}
    boneToModify {boneName=dyn_eye_r}
    rotation {yaw=-20}
   }
   n11 = Get Active_Kill_B() -> (Active_Kill_B:bool)
   n12 = "Transform (Modify) Bone - Bone: dyn_eye_l"!(ComponentPose=@n10, bAlphaBoolEnabled:bool=@n11, Translation="0.35,-0.2,0.2", Rotation="0,0,0", Scale="1.2,1.2,1.2") -> (Pose) {
    alphaInputType=Bool
    rotationMode=BMM_Additive
    rotationSpace=BCS_ParentBoneSpace
    scaleMode=BMM_Additive
    scaleSpace=BCS_BoneSpace
    translationMode=BMM_Additive
    translationSpace=BCS_BoneSpace
    alphaBoolBlend {blendInTime=0.3, blendOption=Linear, blendOutTime=0.3}
    boneToModify {boneName=dyn_eye_l}
    scale=<1.2 1.2 1.2>
    translation=<0.35 -0.2 0.2>
   }
   n13 = "Transform (Modify) Bone - Bone: dyn_eye_l"(ComponentPose=@n12, bAlphaBoolEnabled=@n9, Translation="0,0,0", Rotation="Get rotate  Eye  L"() -> ("rotate  Eye  L":Rotator), Scale="1,1,1") {
    alphaInputType=Bool
    rotationMode=BMM_Additive
    rotationSpace=BCS_ParentBoneSpace
    scaleMode=BMM_Additive
    scaleSpace=BCS_BoneSpace
    translationMode=BMM_Additive
    translationSpace=BCS_BoneSpace
    alphaBoolBlend {blendInTime=0.3, blendOption=Cubic, blendOutTime=0.5}
    boneToModify {boneName=dyn_eye_l}
    rotation {yaw=-20}
   }
   n14 = "Transform (Modify) Bone - Bone: dyn_eye_r"(ComponentPose=@n13, bAlphaBoolEnabled=@n11, Translation="0.35,0.2,-0.2", Rotation="0,0,0", Scale="1.2,1.2,1.2") {
    alphaInputType=Bool
    rotationMode=BMM_Additive
    rotationSpace=BCS_ParentBoneSpace
    scaleMode=BMM_Additive
    scaleSpace=BCS_BoneSpace
    translationMode=BMM_Additive
    translationSpace=BCS_BoneSpace
    alphaBoolBlend {blendInTime=0.2, blendOption=Linear, blendOutTime=0.4}
    boneToModify {boneName=dyn_eye_r}
    scale=<1.2 1.2 1.2>
    translation=<0.35 0.2 -0.2>
   }
   Output Pose(Result=Component To Local(ComponentPose=@n14)) {layerGroup=None}
   Get Active_Kill_B()
  ]}
 EventGraph {nodes [
   Event BlueprintUpdateAnimation()
   n1 = GetGameTimeInSeconds()
   n2 = float * float(A=@n1, B=680)
   > "Set rotate  Eye  R"("rotate  Eye  R":Rotator=MakeRotator(Roll=NormalizeAxis!(Angle:float=@n2) -> (ReturnValue:float))) -> (Output_Get:Rotator)
   n3 = float * float(A=@n1, B=-630)
   > "Set rotate  Eye  L"("rotate  Eye  L":Rotator=MakeRotator(Roll=NormalizeAxis(Angle=@n3))) -> (Output_Get:Rotator)
   n4 = float * float(A="Sin (Radians)"(A=float * float(A=@n1, B=10)), B=30)
   > "Set rotate  Head"("rotate  Head":Rotator=MakeRotator(Yaw=@n4)) -> (Output_Get:Rotator)
   > Get!(execute:Exec) -> (then:Exec, else:Exec, Skel_Mesh:Skeletal Mesh Component Object Reference)
   > Sequence!(execute) -> (then_0, then_1, then_2:Exec)
   > Branch(Condition=Get Active_Kill())
   > Delay(Duration=0.5)
   > Set Active_Kill(Active_Kill:bool=false) -> (Output_Get:bool)
   > Delay(Duration=1.75)
   > Set Active_Kill_B(Active_Kill_B:bool=false) -> (Output_Get:bool)
   > Set Active_Kill_C(Active_Kill_C:bool=false) -> (Output_Get:bool)
   Event OnCharacterCustomizationLoaded()
   > n5 = Cast To FortPlayerPawn(Object=TryGetFortPawn())
   > n6 = WatchCosmeticStat!(
    execute:Exec
    Pawn:Fort Player Pawn Object Reference=@n5
    StatTag:Fort Stat Manager Tag Structure=(TagName="StatManager.AthenaKills")
    DefaultValue:int=0
    bBroadcastInitialValue:bool=true
    InitialValueBroadcastDelay:float=-1
    ChangeValueBroadcastDelay:float=-1
    bConsumeWaitingNotifications:bool=true
   ) -> (then:Exec, OnStatChanged:Exec, StatName:Name, StatValue:int) {is=title K2Node_AsyncAction}
   Set KillCount(execute=@n6.OnStatChanged, KillCount:double="To Float (Integer)"(InInt=@n6.StatValue)) -> (Output_Get:double)
   > Set Active_Kill(Active_Kill=true)
   > Set Active_Kill_B(Active_Kill_B=true)
   > Set Active_Kill_C(Active_Kill_C=true)
   Event BlueprintInitializeAnimation()
   > Set Skel_Mesh(Skel_Mesh=GetOwnerSkeletalMeshComponent())
   PrintString!(execute, InString=- - - Kill - - -, bPrintToScreen, bPrintToLog, TextColor, Duration, Key) -> (then)
   TryGetPawnOwner()
   NormalizeAxis(Angle=@n4)
   Delay(Duration=0.2)
  ]}
]
variables [
 Active_Kill {type=boolean}
 Active_Kill_B {type=boolean}
 Active_Kill_C {type=boolean}
 KillCount {type=number}
 Pawn {type=FortPlayerPawn}
 "rotate  Eye  L" {type=Rotator}
 "rotate  Eye  R" {type=Rotator}
 "rotate  Head" {type=Rotator}
 "rotate  Head  B" {type=Rotator}
 Skel_Mesh {type=SkeletalMeshComponent}
]
